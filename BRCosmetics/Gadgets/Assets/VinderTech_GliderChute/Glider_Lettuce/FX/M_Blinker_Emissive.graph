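# Unlit translucent particle material in the Lettuce glider's FX folder, apparently for an emissive blinker effect.
# It is flagged for use with Niagara mesh particles.
# Its only expression is a Particle Color node with no connections shown, so any color or blink would depend on the particle system rather than logic in the material.
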Material /BRCosmetics/Gadgets/Assets/VinderTech_GliderChute/Glider_Lettuce/FX/M_Blinker_Emissive
graph {
 properties {
  bUsedWithNiagaraMeshParticles=true
  blendMode=BLEND_Translucent
  pixelDepthOffsetMode=PDOM_Legacy
  shadingModel=MSM_Unlit
  usedShadingModels=Unlit
 }
 nodes [ParticleColor()]
}
