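# Niagara system for the Lettuce glider's effects: turn blinkers, timed breakaway paper pieces, toon exhaust smoke, speed lines and V-Bucks.
# Emitters sample the user skeletal mesh, with paper bursts staggered by part and smoke alternating between two sockets.
# Steering dot products drive blinker alpha, smoke intensity and speed-line fade.
# Front-end flags and modifiers retune spawn rates, velocities and timing for menu previews versus gameplay.
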NiagaraSystem /BRCosmetics/Gadgets/Assets/VinderTech_GliderChute/Glider_Lettuce/FX/NS_Glider_Lettuce
dependencies {
 usedDataInterfaces [
  &NiagaraDataInterfaceCurve
  &NiagaraDataInterfaceArrayColor
  &NiagaraDataInterfaceMeshRendererInfo
  &NiagaraDataInterfaceArrayFloat3
  &NiagaraDataInterfaceArrayFloat
  &NiagaraDataInterfaceColorCurve
  &NiagaraDataInterfaceVector2DCurve
 ]
 usedDynamicInputs [
  &MakeFloatFromVector
  /Niagara/DynamicInputs/Math/Min_Float
  &ClampFloat
  &Subtract_Float
  &Add_Float
  &Power
  &Abs_Float
 ]
 usedModules [
  &SystemState
  &self:SystemSpawnScript.SystemScriptSource.SystemScriptGraph.NiagaraNodeAssignment_0.SetVariables_#1
  &Timeline
  &self:SystemSpawnScript.SystemScriptSource.SystemScriptGraph.NiagaraNodeAssignment_1.SetVariables_#2
  &self:SystemSpawnScript.SystemScriptSource.SystemScriptGraph.NiagaraNodeAssignment_3.SetVariables_#3
  &self:SystemSpawnScript.SystemScriptSource.SystemScriptGraph.NiagaraNodeAssignment_2.SetVariables_#4
 ]
}
emitters {
 Back_Blinkers {
  emitterUpdateScript [
   &EmitterState(Life Cycle Mode=::NewEnumerator1, Loop Behavior=::NewEnumerator1, Loop Duration=2, UseLoopDelay=false)
   &SpawnBurst_Instantaneous(Use Spawn Probability=false, UseLoopCountLimit=false)
  ]
  particleSpawnScript [&InitializeParticle(
    Color=<rgba 10 0 0 1>
    Lifetime=0.6
    Sprite Size=<40 15>
    Sprite Size Mode=::NewEnumerator4
    UsePositionOffset=false
    Write Material Random=true
   )]
  particleUpdateScript [
   &ParticleState(Kill Particles When Lifetime Has Elapsed=false, Loop Particles Lifetime=true)
   &ScaleColor(
    Scale Alpha=&FloatFromCurve(
     FloatCurve=DataInterface() {propertyValues {Curve {keys [
         {arriveTangent=9.0725, time=0.24, value=0}
         {arriveTangent=1.3333, time=0.25, value=10}
         <at 0.75 10>
         {leaveTangent=-0.0405, time=0.76, value=0}
        ]}}}
     Scale Curve=&RemapRange(Input Value To Remap=$User.RightDot, Input Min=0.5, Input Max=1, Output Min=0, Output Max=1)
    )
    ScaleA=true
    ScaleRGB=false
   )
   &SolveForcesAndVelocity(Clamp Velocity=false, Limit Acceleration=false)
   &SkeletalMeshLocation(
    Bone / Socket Index=5
    MaskSpawnGroup=false
    Mesh Sampling Type=::NewEnumerator5
    Skeletal Mesh=$User.SkeletalMesh
    Socket Sampling Mode=::NewEnumerator5
    Write To Particle Position=false
   )
   &SkeletalMeshLocation(
    Bone / Socket Index=4
    MaskSpawnGroup=false
    Mesh Sampling Type=::NewEnumerator5
    Orientation Sampling=::NewEnumerator1
    Sampled Position Offset=<0 1 0>
    Skeletal Mesh=$User.SkeletalMesh
    Socket Sampling Mode=::NewEnumerator5
   ) {module=SkeletalMeshLocation001}
   &AlignSpriteToMeshOrientation(SetAlignment=true, SetFacing=true)
   /SetVariables(Particles.Color Core:LinearColor=&MultiplyLinearColor(Linear Color=$Particles.Color, Scale Factor=<1 1 1 0.05>)) {isSetParameters=true, module=SetVariables_#5}
  ]
  rendererClasses [&NiagaraSpriteRendererProperties, &NiagaraMeshRendererProperties]
  renderers [
   {rendererClass=&NiagaraSpriteRendererProperties}
   {rendererClass=&NiagaraMeshRendererProperties, rendererIndex=1}
  ]
 }
 Front_Blinkers {
  emitterUpdateScript [
   &EmitterState(Life Cycle Mode=::NewEnumerator1, Loop Behavior=::NewEnumerator1, Loop Duration=2, UseLoopDelay=false)
   &SpawnBurst_Instantaneous(Use Spawn Probability=false, UseLoopCountLimit=false)
  ]
  particleSpawnScript [&InitializeParticle(
    Color=<rgba 10 0 0 1>
    Lifetime=0.6
    Sprite Size=<15 7>
    Sprite Size Mode=::NewEnumerator4
    UsePositionOffset=false
    Write Material Random=true
   )]
  particleUpdateScript [
   &ParticleState(Kill Particles When Lifetime Has Elapsed=false, Loop Particles Lifetime=true)
   &ScaleColor(
    Scale Alpha=&FloatFromCurve(
     FloatCurve=DataInterface() {propertyValues {Curve {keys [
         {arriveTangent=9.0725, time=0.24, value=0}
         {arriveTangent=1.3333, time=0.25, value=7.5}
         <at 0.75 7.5>
         {leaveTangent=-0.0405, time=0.76, value=0}
        ]}}}
     Scale Curve=&RemapRange(Input Value To Remap=$User.RightDot, Input Min=0.5, Input Max=1, Output Min=0, Output Max=1)
    )
    ScaleA=true
    ScaleRGB=false
   )
   &SolveForcesAndVelocity(Clamp Velocity=false, Limit Acceleration=false)
   &SkeletalMeshLocation(
    Bone / Socket Index=6
    MaskSpawnGroup=false
    Mesh Sampling Type=::NewEnumerator5
    Skeletal Mesh=$User.SkeletalMesh
    Socket Sampling Mode=::NewEnumerator5
    Write To Particle Position=false
   )
   &SkeletalMeshLocation(
    Bone / Socket Index=4
    MaskSpawnGroup=false
    Mesh Sampling Type=::NewEnumerator5
    Orientation Sampling=::NewEnumerator1
    Sampled Position Offset=<0 -1 0>
    Skeletal Mesh=$User.SkeletalMesh
    Socket Sampling Mode=::NewEnumerator5
   ) {module=SkeletalMeshLocation001}
   &AlignSpriteToMeshOrientation(SetAlignment=true, SetFacing=true)
   /SetVariables(Particles.Color Core:LinearColor=&MultiplyLinearColor(Linear Color=$Particles.Color, Scale Factor=<1 1 1 0.05>)) {isSetParameters=true, module=SetVariables_#5}
  ]
  rendererClasses [&NiagaraSpriteRendererProperties, &NiagaraMeshRendererProperties]
  renderers [
   {rendererClass=&NiagaraSpriteRendererProperties}
   {rendererClass=&NiagaraMeshRendererProperties, rendererIndex=1}
  ]
 }
 Paper_Pieces_Break {
  emitterSpawnScript [
   /Game/Effects/Niagara/MasterMaterials/ToonShader/Template/NM_ToonShaderParameters()
   /Game/Effects/Niagara/MasterMaterials/ToonShader/Template/NM_OutlineParameters()
  ]
  emitterUpdateScript [
   &EmitterState(Life Cycle Mode=::NewEnumerator1, Loop Behavior=::NewEnumerator1, UseLoopDelay=false)
   &SpawnBurst_Instantaneous(Spawn Time=1.7, Use Spawn Probability=false, UseLoopCountLimit=false)
  ]
  particleSpawnScript [
   &InitializeParticle(
    Lifetime=&RandomRangeFloat(Minimum=1, Maximum=1.3)
    Mesh Scale Mode=::NewEnumerator3
    Mesh Uniform Scale=0.75
    UsePositionOffset=false
    Write Material Random=true
   )
   &SkeletalMeshLocation(
    Bone / Socket Index=4
    MaskSpawnGroup=false
    Mesh Sampling Type=::NewEnumerator5
    Skeletal Mesh=$User.SkeletalMesh
    Socket Sampling Mode=::NewEnumerator5
   )
   &AddVelocity(
    Velocity=<1 0.2 0>
    Velocity Speed Scale=&FN_FrontEndFloatModifier(Gameplay=1000, Front End=&RandomRangeFloat(Minimum=1000, Maximum=1500))
   )
   &AddVelocity(
    Velocity=<0 1 0>
    Velocity Speed Scale=&FN_FrontEndFloatModifier(Gameplay=0, Front End=&RandomRangeFloat(Minimum=3500, Maximum=4000))
   ) {module=AddVelocity001}
  ]
  particleUpdateScript [
   &ParticleState()
   &GravityForce()
   &UpdateMeshOrientation(
    Rotation Coordinate Space=::NewEnumerator0
    Rotation Rate=&FloatFromCurve(FloatCurve=DataInterface() {propertyValues {Curve {keys [
         {interpMode=RCIM_Cubic, tangentMode=RCTM_User, time=0, value=0}
         {interpMode=RCIM_Cubic, tangentMode=RCTM_User, time=0.5, value=1}
         {tangentMode=RCTM_User, time=1, value=0}
        ]}}})
   )
   &ScaleMeshSize(Scale Factor=&VectorFromFloat(Value=&FloatFromCurve(FloatCurve=DataInterface() {propertyValues {Curve {keys [
          {tangentMode=RCTM_User, time=0, value=1}
          {arriveTangent=-0.0405, interpMode=RCIM_Cubic, leaveTangent=-0.0405, tangentMode=RCTM_User, time=0.2, value=1}
          {arriveTangent=-2.408, interpMode=RCIM_Cubic, leaveTangent=-2.408, tangentMode=RCTM_User, time=1, value=0}
         ]}}})))
   &AerodynamicDrag(
    Aerodynamic Drag=0.5
    Aerodynamic Rotational Drag=2
    Pivot Offset Maximum:Vector3f=<0.1 0.1 0.1>
    Pivot Offset Minimum:Vector3f=<-0.1 -0.1 -0.1>
    Pivot Offset Minimum Length:Vector3f=<0 0 0>
    Rotation Strength:NiagaraFloat=&FN_FrontEndFloatModifier(Gameplay=&RandomRangeFloat(Minimum=1, Maximum=3), Front End=0)
    Use Linear Drag=true
    Use Rotational Drag=true
   ) {enabled=false}
   &InheritSourceMovement(
    Applied Position Scale=&VectorFromFloat(Value=&FloatFromCurve(FloatCurve=DataInterface() {propertyValues {Curve {keys [
          {arriveTangent=-1.813, interpMode=RCIM_Cubic, leaveTangent=-1.813, tangentMode=RCTM_User, time=0, value=1}
          {interpMode=RCIM_Cubic, tangentMode=RCTM_User, time=1, value=0}
         ]}}}))
    ApplyForce=false
    ApplyPosition=true
    ApplyPositionFalloff=false
   )
   &Drag(
    Drag=&FN_FrontEndFloatModifier(Gameplay=5, Front End=2)
    Use Linear Drag=true
    Use Rotational Drag=true
   )
   &CurlNoiseForce(Pan Noise=false)
   &SolveForcesAndVelocity(Clamp Velocity=false, Limit Acceleration=false)
   &AlignSpriteToMeshOrientation(SetAlignment=true, SetFacing=true)
  ]
  rendererClasses [&NiagaraMeshRendererProperties]
  renderers [
   {rendererClass=&NiagaraMeshRendererProperties, rendererIndex=0}
   {rendererClass=&NiagaraMeshRendererProperties, rendererIndex=1}
  ]
 }
 Paper_Pieces_Door_Right {
  emitterSpawnScript [
   &NM_ToonShaderParameters()
   &NM_OutlineParameters()
  ]
  emitterUpdateScript [
   &EmitterState(Life Cycle Mode=::NewEnumerator1, Loop Behavior=::NewEnumerator1, UseLoopDelay=false)
   &SpawnBurst_Instantaneous(Spawn Time=1.42, Use Spawn Probability=false, UseLoopCountLimit=false)
  ]
  particleSpawnScript [
   &InitializeParticle(
    Lifetime=&RandomRangeFloat(Minimum=1, Maximum=1.3)
    Mesh Scale Mode=::NewEnumerator3
    Mesh Uniform Scale=0.75
    UsePositionOffset=false
    Write Material Random=true
   )
   &SkeletalMeshLocation(
    Bone / Socket Index=4
    MaskSpawnGroup=false
    Mesh Sampling Type=::NewEnumerator5
    Sampled Position Offset=<-1 0 0>
    Skeletal Mesh=$User.SkeletalMesh
    Socket Sampling Mode=::NewEnumerator5
   )
   &AddVelocity(
    Velocity=<1 0 0.1>
    Velocity Speed Scale=&FN_FrontEndFloatModifier(Gameplay=1000, Front End=&RandomRangeFloat(Minimum=2000, Maximum=2500))
   )
   &AddVelocity(
    Velocity=<0 1 0>
    Velocity Speed Scale=&FN_FrontEndFloatModifier(Gameplay=0, Front End=&RandomRangeFloat(Minimum=3500, Maximum=4000))
   ) {module=AddVelocity001}
  ]
  particleUpdateScript [
   &ParticleState()
   &GravityForce()
   &UpdateMeshOrientation(
    Rotation Coordinate Space=::NewEnumerator0
    Rotation Rate=&FloatFromCurve(FloatCurve=DataInterface() {propertyValues {Curve {keys [
         {interpMode=RCIM_Cubic, tangentMode=RCTM_User, time=0, value=0}
         {interpMode=RCIM_Cubic, tangentMode=RCTM_User, time=0.5, value=1}
         {tangentMode=RCTM_User, time=1, value=0}
        ]}}})
    Rotation Vector=&RandomVector(Vector Scale=1)
   )
   &ScaleMeshSize(Scale Factor=&VectorFromFloat(Value=&FloatFromCurve(FloatCurve=DataInterface() {propertyValues {Curve {keys [
          {tangentMode=RCTM_User, time=0, value=1}
          {arriveTangent=-0.0405, interpMode=RCIM_Cubic, leaveTangent=-0.0405, tangentMode=RCTM_User, time=0.7009, value=1}
          {arriveTangent=-8.0089, interpMode=RCIM_Cubic, leaveTangent=-8.0089, tangentMode=RCTM_User, time=1, value=0}
         ]}}})))
   &AerodynamicDrag(
    Aerodynamic Drag=0.5
    Aerodynamic Rotational Drag=2
    Drag Shape Curve=DataInterface() {propertyValues {Curve {keys [
        {interpMode=RCIM_Cubic, tangentMode=RCTM_User, time=0, value=0.1}
        {arriveTangent=2.0799, interpMode=RCIM_Cubic, leaveTangent=2.0799, tangentMode=RCTM_User, time=1, value=1}
       ]}}}
    Lift Shape Curve=DataInterface() {propertyValues {Curve {keys [
        {interpMode=RCIM_Cubic, tangentMode=RCTM_User, time=0, value=0}
        {interpMode=RCIM_Cubic, tangentMode=RCTM_User, time=0.5, value=1}
        {tangentMode=RCTM_User, time=1, value=0}
       ]}}}
    Pivot Offset Maximum:Vector3f=<0.1 0.1 0.1>
    Pivot Offset Minimum:Vector3f=<-0.1 -0.1 -0.1>
    Pivot Offset Minimum Length:Vector3f=<0 0 0>
    Rotation Strength:NiagaraFloat=&FN_FrontEndFloatModifier(Gameplay=&RandomRangeFloat(Minimum=1, Maximum=1.5), Front End=0)
    Use Linear Drag=true
    Use Rotational Drag=true
   ) {enabled=false}
   &InheritSourceMovement(
    Applied Position Scale=&VectorFromFloat(Value=&FloatFromCurve(FloatCurve=DataInterface() {propertyValues {Curve {keys [
          {arriveTangent=-3.1692, interpMode=RCIM_Cubic, leaveTangent=-3.1692, tangentMode=RCTM_User, time=0, value=1}
          {interpMode=RCIM_Cubic, tangentMode=RCTM_User, time=1, value=0}
         ]}}}))
    ApplyForce=false
    ApplyPosition=true
    ApplyPositionFalloff=false
   )
   &Drag(
    Drag=&FN_FrontEndFloatModifier(Gameplay=4, Front End=2)
    Use Linear Drag=true
    Use Rotational Drag=true
   )
   &CurlNoiseForce(Pan Noise=false)
   &SolveForcesAndVelocity(Clamp Velocity=false, Limit Acceleration=false)
   &AlignSpriteToMeshOrientation(SetAlignment=true, SetFacing=true)
  ]
  rendererClasses [&NiagaraMeshRendererProperties]
  renderers [
   {rendererClass=&NiagaraMeshRendererProperties, rendererIndex=0}
   {rendererClass=&NiagaraMeshRendererProperties, rendererIndex=1}
  ]
 }
 Paper_Pieces_Exhaust_L {
  emitterSpawnScript [
   &NM_ToonShaderParameters()
   &NM_OutlineParameters()
  ]
  emitterUpdateScript [
   &EmitterState(Life Cycle Mode=::NewEnumerator1, Loop Behavior=::NewEnumerator1, UseLoopDelay=false)
   &SpawnBurst_Instantaneous(Spawn Time=2, Use Spawn Probability=false, UseLoopCountLimit=false)
  ]
  particleSpawnScript [
   &InitializeParticle(
    Lifetime=&RandomRangeFloat(Minimum=1, Maximum=1.3)
    Mesh Scale Mode=::NewEnumerator3
    Mesh Uniform Scale=0.75
    UsePositionOffset=false
    Write Material Random=true
   )
   &SkeletalMeshLocation(
    Bone / Socket Index=4
    MaskSpawnGroup=false
    Mesh Sampling Type=::NewEnumerator5
    Sampled Position Offset=<1 1 0>
    Skeletal Mesh=$User.SkeletalMesh
    Socket Sampling Mode=::NewEnumerator5
   )
   &AddVelocity(Velocity=<0 0 -1>, Velocity Speed Scale=&RandomRangeFloat(Minimum=750, Maximum=1000))
   &AddVelocity(
    Velocity=<0 1 -0.1>
    Velocity Speed Scale=&FN_FrontEndFloatModifier(Gameplay=0, Front End=&RandomRangeFloat(Minimum=3500, Maximum=4000))
   ) {enabled=false, module=AddVelocity001}
  ]
  particleUpdateScript [
   &ParticleState()
   &GravityForce()
   &GravityForce(Gravity=&RotateVectorByQuaternion(
     Vector To Rotate=&Multiply_VectorByFloat(
      Vector=<0 1 0>
      Float=&FloatFromCurve(
       FloatCurve=DataInterface() {propertyValues {Curve {keys [
           {arriveTangent=23.9172, interpMode=RCIM_Cubic, leaveTangent=23.9172, tangentMode=RCTM_User, time=0, value=0}
           {arriveTangent=0.0112, interpMode=RCIM_Cubic, leaveTangent=0.0112, tangentMode=RCTM_User, time=0.05, value=1.6}
           {arriveTangent=-0.0134, interpMode=RCIM_Cubic, leaveTangent=-0.0134, tangentMode=RCTM_User, time=1, value=0}
          ]}}}
       Scale Curve=&FN_FrontEndFloatModifier(Gameplay=0, Front End=7200)
      )
     )
     Quaternion=$Particles.SkeletalMeshLocation.SampledRotation
    )) {module=GravityForce001}
   &UpdateMeshOrientation(
    Rotation Coordinate Space=::NewEnumerator0
    Rotation Rate=&FloatFromCurve(
     FloatCurve=DataInterface() {propertyValues {Curve {keys [
         {interpMode=RCIM_Cubic, tangentMode=RCTM_User, time=0, value=0}
         {interpMode=RCIM_Cubic, tangentMode=RCTM_User, time=0.5, value=1}
         {tangentMode=RCTM_User, time=1, value=0}
        ]}}}
     Scale Curve=1.5
    )
    Rotation Vector=&RandomVector(Vector Scale=1)
   )
   &ScaleMeshSize(Scale Factor=&VectorFromFloat(Value=&FloatFromCurve(FloatCurve=DataInterface() {propertyValues {Curve {keys [
          {tangentMode=RCTM_User, time=0, value=1}
          {arriveTangent=-0.0405, interpMode=RCIM_Cubic, leaveTangent=-0.0405, tangentMode=RCTM_User, time=0.2, value=1}
          {arriveTangent=-2.408, interpMode=RCIM_Cubic, leaveTangent=-2.408, tangentMode=RCTM_User, time=1, value=0}
         ]}}})))
   &AerodynamicDrag(
    Aerodynamic Drag=3
    Aerodynamic Rotational Drag=2
    Drag Shape Curve=DataInterface() {propertyValues {Curve {keys [
        {interpMode=RCIM_Cubic, tangentMode=RCTM_User, time=0, value=0.1}
        {arriveTangent=2.0799, interpMode=RCIM_Cubic, leaveTangent=2.0799, tangentMode=RCTM_User, time=1, value=1}
       ]}}}
    Lift Shape Curve=DataInterface() {propertyValues {Curve {keys [
        {interpMode=RCIM_Cubic, tangentMode=RCTM_User, time=0, value=0}
        {interpMode=RCIM_Cubic, tangentMode=RCTM_User, time=0.5, value=1}
        {tangentMode=RCTM_User, time=1, value=0}
       ]}}}
    Pivot Offset Maximum:Vector3f=<0.1 0.1 0.1>
    Pivot Offset Minimum:Vector3f=<-0.1 -0.1 -0.1>
    Pivot Offset Minimum Length:Vector3f=<0 0 0>
    Rotation Strength:NiagaraFloat=&FN_FrontEndFloatModifier(Gameplay=&RandomRangeFloat(Minimum=1, Maximum=1.5), Front End=0)
    Use Linear Drag=true
    Use Rotational Drag=true
   ) {enabled=false}
   &InheritSourceMovement(
    Applied Position Scale=&VectorFromFloat(Value=&FloatFromCurve(FloatCurve=DataInterface() {propertyValues {Curve {keys [
          {arriveTangent=-2.4306, interpMode=RCIM_Cubic, leaveTangent=-2.4306, tangentMode=RCTM_User, time=0, value=1}
          {interpMode=RCIM_Cubic, tangentMode=RCTM_User, time=1, value=0}
         ]}}}))
    ApplyForce=false
    ApplyPosition=true
    ApplyPositionFalloff=false
   )
   &Drag(
    Drag=&FN_FrontEndFloatModifier(Gameplay=4, Front End=4)
    Use Linear Drag=true
    Use Rotational Drag=true
   )
   &CurlNoiseForce(Pan Noise=false)
   &SolveForcesAndVelocity(Clamp Velocity=false, Limit Acceleration=false)
   &AlignSpriteToMeshOrientation(SetAlignment=true, SetFacing=true)
  ]
  rendererClasses [&NiagaraMeshRendererProperties]
  renderers [
   {rendererClass=&NiagaraMeshRendererProperties, rendererIndex=0}
   {rendererClass=&NiagaraMeshRendererProperties, rendererIndex=1}
  ]
 }
 Paper_Pieces_Exhaust_R {
  emitterSpawnScript [
   &NM_ToonShaderParameters()
   &NM_OutlineParameters()
  ]
  emitterUpdateScript [
   &EmitterState(Life Cycle Mode=::NewEnumerator1, Loop Behavior=::NewEnumerator1, UseLoopDelay=false)
   &SpawnBurst_Instantaneous(Spawn Time=2.1, Use Spawn Probability=false, UseLoopCountLimit=false)
  ]
  particleSpawnScript [
   &InitializeParticle(
    Lifetime=&RandomRangeFloat(Minimum=1, Maximum=1.3)
    Mesh Scale Mode=::NewEnumerator3
    Mesh Uniform Scale=0.75
    UsePositionOffset=false
    Write Material Random=true
   )
   &SkeletalMeshLocation(
    Bone / Socket Index=4
    MaskSpawnGroup=false
    Mesh Sampling Type=::NewEnumerator5
    Sampled Position Offset=<0 1 -1>
    Skeletal Mesh=$User.SkeletalMesh
    Socket Sampling Mode=::NewEnumerator5
   )
   &AddVelocity(Velocity=<0 0 -1>, Velocity Speed Scale=&RandomRangeFloat(Minimum=750, Maximum=1000))
   &AddVelocity(
    Velocity=<0 1 -0.1>
    Velocity Speed Scale=&FN_FrontEndFloatModifier(Gameplay=0, Front End=&RandomRangeFloat(Minimum=3500, Maximum=4000))
   ) {enabled=false, module=AddVelocity001}
  ]
  particleUpdateScript [
   &ParticleState()
   &GravityForce()
   &GravityForce(Gravity=&RotateVectorByQuaternion(
     Vector To Rotate=&Multiply_VectorByFloat(
      Vector=<0 1 0>
      Float=&FloatFromCurve(
       FloatCurve=DataInterface() {propertyValues {Curve {keys [
           {arriveTangent=26.9415, interpMode=RCIM_Cubic, leaveTangent=26.9415, tangentMode=RCTM_User, time=0, value=0}
           {arriveTangent=0.0112, interpMode=RCIM_Cubic, leaveTangent=0.0112, tangentMode=RCTM_User, time=0.05, value=1.6}
           {arriveTangent=-0.0134, interpMode=RCIM_Cubic, leaveTangent=-0.0134, tangentMode=RCTM_User, time=1, value=0}
          ]}}}
       Scale Curve=&FN_FrontEndFloatModifier(Gameplay=0, Front End=7500)
      )
     )
     Quaternion=$Particles.SkeletalMeshLocation.SampledRotation
    )) {module=GravityForce001}
   &UpdateMeshOrientation(
    Rotation Rate=&FloatFromCurve(
     FloatCurve=DataInterface() {propertyValues {Curve {keys [
         {interpMode=RCIM_Cubic, tangentMode=RCTM_User, time=0, value=0}
         {interpMode=RCIM_Cubic, tangentMode=RCTM_User, time=0.5, value=1}
         {tangentMode=RCTM_User, time=1, value=0}
        ]}}}
     Scale Curve=1.5
    )
    Rotation Vector=&RandomVector(Vector Scale=1)
   )
   &ScaleMeshSize(Scale Factor=&VectorFromFloat(Value=&FloatFromCurve(FloatCurve=DataInterface() {propertyValues {Curve {keys [
          {tangentMode=RCTM_User, time=0, value=1}
          {arriveTangent=-0.0405, interpMode=RCIM_Cubic, leaveTangent=-0.0405, tangentMode=RCTM_User, time=0.2, value=1}
          {arriveTangent=-2.408, interpMode=RCIM_Cubic, leaveTangent=-2.408, tangentMode=RCTM_User, time=1, value=0}
         ]}}})))
   &AerodynamicDrag(
    Aerodynamic Drag=5
    Aerodynamic Rotational Drag=2
    Drag Shape Curve=DataInterface() {propertyValues {Curve {keys [
        {interpMode=RCIM_Cubic, tangentMode=RCTM_User, time=0, value=0.1}
        {arriveTangent=2.0799, interpMode=RCIM_Cubic, leaveTangent=2.0799, tangentMode=RCTM_User, time=1, value=1}
       ]}}}
    Lift Shape Curve=DataInterface() {propertyValues {Curve {keys [
        {interpMode=RCIM_Cubic, tangentMode=RCTM_User, time=0, value=0}
        {interpMode=RCIM_Cubic, tangentMode=RCTM_User, time=0.5, value=1}
        {tangentMode=RCTM_User, time=1, value=0}
       ]}}}
    Pivot Offset Maximum:Vector3f=<0.1 0.1 0.1>
    Pivot Offset Minimum:Vector3f=<-0.1 -0.1 -0.1>
    Pivot Offset Minimum Length:Vector3f=<0 0 0>
    Rotation Strength:NiagaraFloat=&FN_FrontEndFloatModifier(Gameplay=&RandomRangeFloat(Minimum=1, Maximum=2), Front End=0)
    Use Linear Drag=true
    Use Rotational Drag=true
   ) {enabled=false}
   &InheritSourceMovement(
    Applied Position Scale=&VectorFromFloat(Value=&FloatFromCurve(FloatCurve=DataInterface() {propertyValues {Curve {keys [
          {arriveTangent=-2.5947, interpMode=RCIM_Cubic, leaveTangent=-2.5947, tangentMode=RCTM_User, time=0, value=1}
          {interpMode=RCIM_Cubic, tangentMode=RCTM_User, time=1, value=0}
         ]}}}))
    ApplyForce=false
    ApplyPosition=true
    ApplyPositionFalloff=false
   )
   &Drag(
    Drag=&FN_FrontEndFloatModifier(Gameplay=4, Front End=4)
    Use Linear Drag=true
    Use Rotational Drag=true
   )
   &CurlNoiseForce(Pan Noise=false)
   &SolveForcesAndVelocity(Clamp Velocity=false, Limit Acceleration=false)
   &AlignSpriteToMeshOrientation(SetAlignment=true, SetFacing=true)
  ]
  rendererClasses [&NiagaraMeshRendererProperties]
  renderers [
   {rendererClass=&NiagaraMeshRendererProperties, rendererIndex=0}
   {rendererClass=&NiagaraMeshRendererProperties, rendererIndex=1}
  ]
 }
 Paper_Pieces_Headlight {
  emitterSpawnScript [
   &NM_ToonShaderParameters()
   &NM_OutlineParameters()
  ]
  emitterUpdateScript [
   &EmitterState(Life Cycle Mode=::NewEnumerator1, Loop Behavior=::NewEnumerator1, UseLoopDelay=false)
   &SpawnBurst_Instantaneous(Spawn Time=0.535, Use Spawn Probability=false, UseLoopCountLimit=false)
  ]
  particleSpawnScript [
   &InitializeParticle(
    Lifetime=&RandomRangeFloat(Minimum=1, Maximum=1.3)
    Mesh Scale Mode=::NewEnumerator3
    Mesh Uniform Scale=0.75
    UsePositionOffset=false
    Write Material Random=true
   )
   &SkeletalMeshLocation(
    Bone / Socket Index=4
    MaskSpawnGroup=false
    Mesh Sampling Type=::NewEnumerator5
    Sampled Position Offset=<-1 0 1>
    Skeletal Mesh=$User.SkeletalMesh
    Socket Sampling Mode=::NewEnumerator5
   )
   &AddVelocity(
    Velocity=<1 -0.5 0.5>
    Velocity Speed Scale=&FN_FrontEndFloatModifier(Gameplay=750, Front End=&RandomRangeFloat(Minimum=500, Maximum=750))
   )
   &AddVelocity(
    Velocity=<0 1 0>
    Velocity Speed Scale=&FN_FrontEndFloatModifier(Gameplay=0, Front End=&RandomRangeFloat(Minimum=1500, Maximum=2000))
   ) {module=AddVelocity001}
  ]
  particleUpdateScript [
   &ParticleState()
   &GravityForce()
   &UpdateMeshOrientation(
    Rotation Rate=&FloatFromCurve(
     FloatCurve=DataInterface() {propertyValues {Curve {keys [
         {interpMode=RCIM_Cubic, tangentMode=RCTM_User, time=0, value=0}
         {interpMode=RCIM_Cubic, tangentMode=RCTM_User, time=0.5, value=1}
         {tangentMode=RCTM_User, time=1, value=0}
        ]}}}
     Scale Curve=0.5
    )
    Rotation Vector=&RandomVector(Vector Scale=1)
   )
   &ScaleMeshSize(Scale Factor=&VectorFromFloat(Value=&FloatFromCurve(FloatCurve=DataInterface() {propertyValues {Curve {keys [
          {tangentMode=RCTM_User, time=0, value=1}
          {arriveTangent=-0.0405, interpMode=RCIM_Cubic, leaveTangent=-0.0405, tangentMode=RCTM_User, time=0.2, value=1}
          {arriveTangent=-2.408, interpMode=RCIM_Cubic, leaveTangent=-2.408, tangentMode=RCTM_User, time=1, value=0}
         ]}}})))
   &AerodynamicDrag(
    Aerodynamic Drag=0.5
    Aerodynamic Rotational Drag=2
    Drag Shape Curve=DataInterface() {propertyValues {Curve {keys [
        {interpMode=RCIM_Cubic, tangentMode=RCTM_User, time=0, value=0.1}
        {arriveTangent=2.0799, interpMode=RCIM_Cubic, leaveTangent=2.0799, tangentMode=RCTM_User, time=1, value=1}
       ]}}}
    Lift Shape Curve=DataInterface() {propertyValues {Curve {keys [
        {interpMode=RCIM_Cubic, tangentMode=RCTM_User, time=0, value=0}
        {interpMode=RCIM_Cubic, tangentMode=RCTM_User, time=0.5, value=1}
        {tangentMode=RCTM_User, time=1, value=0}
       ]}}}
    Pivot Offset Maximum:Vector3f=<0.1 0.1 0.1>
    Pivot Offset Minimum:Vector3f=<-0.1 -0.1 -0.1>
    Pivot Offset Minimum Length:Vector3f=<0 0 0>
    Rotation Strength:NiagaraFloat=&FN_FrontEndFloatModifier(Gameplay=&RandomRangeFloat(Minimum=1, Maximum=2), Front End=0)
    Use Linear Drag=true
    Use Rotational Drag=true
   ) {enabled=false}
   &InheritSourceMovement(
    Applied Position Scale=&VectorFromFloat(Value=&FloatFromCurve(FloatCurve=DataInterface() {propertyValues {Curve {keys [
          {arriveTangent=-0.5704, interpMode=RCIM_Cubic, leaveTangent=-0.5704, tangentMode=RCTM_User, time=0, value=1}
          {interpMode=RCIM_Cubic, tangentMode=RCTM_User, time=1, value=0.75}
         ]}}}))
    ApplyForce=false
    ApplyPosition=true
    ApplyPositionFalloff=false
   )
   &Drag(
    Drag=&FN_FrontEndFloatModifier(Gameplay=2, Front End=4)
    Use Linear Drag=true
    Use Rotational Drag=true
   )
   &CurlNoiseForce(Pan Noise=false)
   &SolveForcesAndVelocity(Clamp Velocity=false, Limit Acceleration=false)
   &AlignSpriteToMeshOrientation(SetAlignment=true, SetFacing=true)
  ]
  rendererClasses [&NiagaraMeshRendererProperties]
  renderers [
   {rendererClass=&NiagaraMeshRendererProperties, rendererIndex=0}
   {rendererClass=&NiagaraMeshRendererProperties, rendererIndex=1}
  ]
 }
 Smoke_Toon_WorldSpace {
  emitterSpawnScript [
   &NM_ToonShaderParameters(HighLight_Softness=0, Lighting_Hardness=0.8, Metallic_LightingPower=0, RimLight_Color:LinearColor=<rgba 0 0 0 1>, RimLight_Width=0)
   &NM_OutlineParameters()
  ]
  emitterUpdateScript [
   &EmitterState(
    Inactive Response=::NewEnumerator1
    Life Cycle Mode=::NewEnumerator1
    Loop Behavior=::NewEnumerator1
    Loop Delay=0.15
    Loop Duration Mode=::NewEnumerator1
    UseLoopDelay=false
   )
   &SpawnRate(
    Spawn Probability=0.75
    SpawnRate=&FloatFromCurve(
     FloatCurve=DataInterface() {propertyValues {Curve {keys [<at 1.1 0>,<at 1.2 2>,<at 1.3 2>,<at 1.5 0>,<at 1.6 2>,<at 1.8 1>,<at 2.1 1>]}}}
     CurveIndex=$Emitter.Age
     Scale Curve=&Multiply_Float(
      A=&MakeCustomFloatFromBool(Bool=$User.bIsFrontEnd, True Float=0, False Float=150)
      B=&MakeCustomFloatFromBool(Bool=&SetBoolByFloatComparison(A=$System.ForwardAndRightDot, B=0.4), True Float=1.5, False Float=1)
     )
    )
    Use Spawn Probability=true
   )
   &SpawnRate(
    Spawn Probability=0.75
    SpawnRate=&FloatFromCurve(
     FloatCurve=DataInterface() {propertyValues {Curve {keys [<at 0.5 0>,<at 0.6 2>,<at 0.7 2>,<at 0.9 0>,<at 1 2>,<at 1.2 1>,<at 1.5 1>]}}}
     CurveIndex=$Emitter.Age
     Scale Curve=&Multiply_Float(
      A=&MakeCustomFloatFromBool(Bool=$User.bIsFrontEnd, True Float=200, False Float=0)
      B=&MakeCustomFloatFromBool(Bool=&SetBoolByFloatComparison(A=$System.ForwardAndRightDot, B=0.4), True Float=1.5, False Float=1)
     )
    )
    Use Spawn Probability=true
   ) {module=SpawnRate001}
  ]
  particleSpawnScript [
   &FN_SetSpawnGroup(Number of Spawn Groups=2)
   &InitializeParticle(
    Color=&FNTimeOfDayColor(Night Color=<rgba 0.8 0.8633 1 1>, Morning Color=<rgba 1 1 1 1>, Day Color=<rgba 1 1 1 1>, Evening Color=<rgba 0.8 0.8633 1 1>)
    Lifetime=&RandomRangeFloat(Minimum=0.4, Maximum=0.6)
    Mesh Scale=&Multiply_VectorByFloat(
     Vector=&VectorFromFloat(Value=&SelectFloatFromArray(
       Array Sampling Mode=::NewEnumerator0
       Direct Array Index=?
       Float Selection Array=DataInterface() {propertyValues {FloatData [1,0.75,0.6]}}
       Random Seed=0
      ))
     Float=&MakeCustomFloatFromBool(Bool=$User.bIsFrontEnd, True Float=0.65, False Float=1)
    )
    Mesh Scale Mode=::NewEnumerator4
    UsePositionOffset=false
    Write Material Random=true
   )
   &SkeletalMeshLocation(
    Bone / Socket Index=&Add_Integer(A=$Particles.SpawnGroup, B=2)
    MaskSpawnGroup=false
    Mesh Sampling Type=::NewEnumerator5
    Orientation Sampling=::NewEnumerator1
    Skeletal Mesh=$User.SkeletalMesh
    Socket Sampling Mode=::NewEnumerator5
   )
   &InitialMeshOrientation(
    Mesh Orientation Mode=::NewEnumerator7
    Rotation=&MakeVector(X=&RandomRangeFloat(Minimum=0, Maximum=1), Y=0, Z=0)
    Use Rotation Vector=true
   )
   &AddVelocity(
    Cone Angle=&MakeCustomFloatFromBool(Bool=&SetBoolByFloatComparison(A=$System.ForwardAndRightDot, B=0.4), True Float=5, False Float=20)
    Cone Axis=&RotateVectorByQuaternion(Vector To Rotate=<0 1 0>, Quaternion=$Particles.SkeletalMeshLocation.SampledRotation)
    Module.UseConeFalloff=false
    Module.UseVelDistribution=false
    Override Seed=false
    Rotation Coordinate Space=::World
    Velocity Mode=::NewEnumerator2
    Velocity Speed=&Multiply_Float(
     A=&RandomRangeFloat(Minimum=275, Maximum=300)
     B=&MakeCustomFloatFromBool(Bool=&SetBoolByFloatComparison(A=$System.ForwardAndRightDot, B=0.4), True Float=1.25, False Float=1)
    )
   )
   /SetVariables(Particles.Gnomon Scale:Vector3f=<1.5 1.5 1.5>) {isSetParameters=true}
  ]
  particleUpdateScript [
   &ParticleState()
   &ScaleMeshSize(Scale Factor=&VectorFromFloat(Value=&FloatFromCurve(FloatCurve=DataInterface() {propertyValues {Curve {keys [
          {arriveTangent=1.9405, interpMode=RCIM_Cubic, leaveTangent=1.9405, tangentMode=RCTM_User, time=0, value=0.1}
          {arriveTangent=-0.0088, interpMode=RCIM_Cubic, leaveTangent=-0.0088, tangentMode=RCTM_User, time=0.7, value=1}
          {arriveTangent=-5.5932, interpMode=RCIM_Cubic, leaveTangent=-5.5932, tangentMode=RCTM_User, time=1, value=0}
         ]}}})))
   &ScaleMeshSize(Scale Factor=&MakeVector(
     X=&FloatFromCurve(
      FloatCurve=DataInterface() {propertyValues {Curve {keys [
          {arriveTangent=-8.2221, interpMode=RCIM_Cubic, leaveTangent=-8.2221, tangentMode=RCTM_User, time=0, value=4}
          {interpMode=RCIM_Cubic, tangentMode=RCTM_User, time=0.5, value=1}
         ]}}}
      Scale Curve=&MakeCustomFloatFromBool(Bool=&SetBoolByFloatComparison(A=$System.ForwardAndRightDot, B=0.4), True Float=1.2, False Float=1)
     )
     Y=1
     Z=1
    )) {module=ScaleMeshSize001}
   &Color(
    Color=&ColorFromCurve(ColorCurve=DataInterface() {propertyValues {
       AlphaCurve {keys [<at 0.1 1>]}
       BlueCurve {keys [<at 0 0.2995>, {arriveTangent=0.0577, time=0.5436, value=0.1797}, <at 0.9966 0.076>]}
       GreenCurve {keys [<at 0 0.3668>, {arriveTangent=-0.0961, time=0.5436, value=0.2201}, <at 0.9966 0.076>]}
       RedCurve {keys [<at 0 0.5>, {arriveTangent=-1.0957, time=0.5436, value=0.3}, <at 0.9966 0.1>]}
      }})
    Scale Alpha=&MakeFloatFromLinearColor(LinearColor=$Particles.Initial.Color, Channel:ENiagaraLinearColor_Channels=::NewEnumerator3)
   ) {enabled=false}
   &ScaleColor(
    Scale RGB=&MakeVectorFromLinearColorRGB(Color=&ColorFromCurve(ColorCurve=DataInterface() {propertyValues {
        AlphaCurve {keys [<at 0.1 1>]}
        BlueCurve {keys [<at 0 0.2995>, {arriveTangent=0.0577, time=0.5436, value=0.1797}, <at 0.9966 0.076>]}
        GreenCurve {keys [<at 0 0.3668>, {arriveTangent=-0.0961, time=0.5436, value=0.2201}, <at 0.9966 0.076>]}
        RedCurve {keys [<at 0 0.5>, {arriveTangent=-1.0957, time=0.5436, value=0.3}, <at 0.9966 0.1>]}
       }}))
    ScaleA=true
    ScaleRGB=true
   )
   &Drag(
    Drag=&FloatFromCurve(
     FloatCurve=DataInterface() {propertyValues {Curve {keys [{tangentMode=RCTM_User, time=0, value=0}, {tangentMode=RCTM_User, time=1, value=1}]}}}
     Scale Curve=5
    )
    Use Linear Drag=true
    Use Rotational Drag=true
   )
   &SkeletalMeshLocation(
    Bone / Socket Index=&Add_Integer(A=$Particles.SpawnGroup, B=2)
    MaskSpawnGroup=false
    Mesh Sampling Type=::NewEnumerator5
    Orientation Sampling=::NewEnumerator1
    Position Sampling=::NewEnumerator2
    Skeletal Mesh=$User.SkeletalMesh
    Socket Sampling Mode=::NewEnumerator5
   ) {module=SkeletalMeshLocation001}
   &InheritSourceMovement(
    Applied Force Scale=<-1 -1 1>
    Applied Position Scale=&MakeVector(
     X=&FloatFromCurve(FloatCurve=DataInterface() {propertyValues {Curve {keys [
          {arriveTangent=-0.5713, interpMode=RCIM_Cubic, leaveTangent=-0.5713, tangentMode=RCTM_User, time=0, value=1}
          {arriveTangent=-0.9092, interpMode=RCIM_Cubic, leaveTangent=-0.9092, tangentMode=RCTM_User, time=1, value=0.25}
         ]}}})
     Y=&FloatFromCurve(FloatCurve=DataInterface() {propertyValues {Curve {keys [
          {arriveTangent=-0.5713, interpMode=RCIM_Cubic, leaveTangent=-0.5713, tangentMode=RCTM_User, time=0, value=1}
          {arriveTangent=-0.9092, interpMode=RCIM_Cubic, leaveTangent=-0.9092, tangentMode=RCTM_User, time=1, value=0.25}
         ]}}})
     Z=&FloatFromCurve(FloatCurve=DataInterface() {propertyValues {Curve {keys [
          {interpMode=RCIM_Cubic, tangentMode=RCTM_User, time=0, value=1}
          {arriveTangent=-1.9883, interpMode=RCIM_Cubic, leaveTangent=-1.9883, tangentMode=RCTM_User, time=1, value=0}
         ]}}})
    )
    ApplyForce=false
    ApplyPosition=true
    ApplyPositionFalloff=false
   )
   &SolveForcesAndVelocity(Clamp Velocity=false, Limit Acceleration=false)
   &UpdateMeshOrientation(
    Facing Direction=&NormalizeVector(
     Vector To Normalize=&Add_Vector(
      A=$Particles.Velocity
      B=&Multiply_VectorByFloat(
       Vector=&Multiply_Vector(
        A=&Subtract_Vector(
         A=$Particles.SkeletalMeshLocation001.SampledVelocity
         B=&Multiply_VectorByFloat(Vector=$Engine.Owner.Velocity, Float=0.8)
        )
        B=<1 1 0>
       )
       Float=&FloatFromCurve(
        FloatCurve=DataInterface() {propertyValues {Curve {keys [
            {arriveTangent=13.8082, interpMode=RCIM_Cubic, leaveTangent=13.8082, tangentMode=RCTM_User, time=0, value=0.05}
            {arriveTangent=-2.8462, interpMode=RCIM_Cubic, leaveTangent=-2.8462, tangentMode=RCTM_User, time=0.3, value=0.7}
            {arriveTangent=-0.0405, interpMode=RCIM_Cubic, leaveTangent=-0.0405, tangentMode=RCTM_User, time=1, value=0}
           ]}}}
        Scale Curve=-1
       )
      )
     )
     Safe Normalize=true
     Safe Fallback Threshold=0
     Safe Fallback Vector=<0 0 0.001>
    )
    Facing Interpolation=false
    Orientation Method=::NewEnumerator1
    Override Quat=false
    TransformFacing=false
    TransformUp=false
   ) {module=UpdateMeshOrientation001}
   &UpdateMeshOrientation(
    Rotation Rate=&FloatFromCurve(
     FloatCurve=DataInterface() {propertyValues {Curve {keys [{interpMode=RCIM_Cubic, tangentMode=RCTM_User, time=0.4, value=0}, {interpMode=RCIM_Cubic, tangentMode=RCTM_User, time=1, value=1}]}}}
     Scale Curve=&RandomRangeFloat(Minimum=-2, Maximum=2)
    )
    Rotation Vector=&RandomVector(Vector Scale=1)
   )
  ]
  rendererClasses [&NiagaraMeshRendererProperties]
  renderers [
   {rendererClass=&NiagaraMeshRendererProperties, rendererIndex=0}
   {rendererClass=&NiagaraMeshRendererProperties, rendererIndex=1}
   {rendererClass=&NiagaraMeshRendererProperties, rendererIndex=2}
  ]
 }
 SpeedLines {
  emitterUpdateScript [
   &EmitterState(Life Cycle Mode=::NewEnumerator1, Loop Behavior=::NewEnumerator1, Loop Duration Mode=::NewEnumerator1, UseLoopDelay=false)
   &SpawnRate(
    Spawn Probability=&FloatFromCurve(
     FloatCurve=DataInterface() {propertyValues {Curve {keys [{interpMode=RCIM_Cubic, tangentMode=RCTM_User, time=2.7, value=0}, {interpMode=RCIM_Cubic, tangentMode=RCTM_User, time=3.3, value=1}]}}}
     CurveIndex=$Emitter.Age
    )
    SpawnRate=&MakeCustomFloatFromBool(Bool=$User.bIsFrontEndPreview, True Float=50, False Float=200)
    Use Spawn Probability=true
   ) {module=SpawnRate001}
  ]
  particleSpawnScript [
   &InitializeParticle(
    Color=&MakeLinearColorFromVectorAndFloat(
     "Vector (RGB)"=<1 1 1>
     "Float (Alpha)"=&Multiply_Float(
      A=&UniformRangedFloat(Minimum=0.1, Maximum=0.18)
      B=&MakeCustomFloatFromBool(Bool=$User.bIsFrontEndPreview, True Float=0.5, False Float=1)
     )
    )
    Lifetime Max=0.16
    Lifetime Min=0.1
    Lifetime Mode=::NewEnumerator1
    Sprite Size Max=<3 10>
    Sprite Size Min=<2 2>
    Sprite Size Mode=::NewEnumerator2
    UsePositionOffset=false
    Write Material Random=true
   )
   &SkeletalMeshLocation(
    Bone Sampling Mode=::NewEnumerator4
    MaskSpawnGroup=false
    Orientation Sampling=::NewEnumerator1
    Sampled Position Offset=<0 25 0>
    Skeletal Mesh=$User.SkeletalMesh
   )
   &TorusLocation(
    Align Torus With Axis=true
    Handle Radius=0
    Large Radius:NiagaraFloat=&FN_FrontEndFloatModifier(
     Gameplay=&RandomRangeFloat(Minimum=100, Maximum=250)
     Front End=&RandomRangeFloat(Minimum=80, Maximum=150)
    )
    MaskSpawnGroup=false
    Module.Override Randomness=false
    Offset=<0 -180 0>
    Torus Axis=<0 1 0>
    Torus Origin:NiagaraPosition=&Add_Vector(
     A=Linked() {linkedVariable {name=Particles.Position, type=NiagaraPosition}}
     B=&MakeVector(
      X=0
      Y=&MakeCustomFloatFromBool(Bool=$User.bIsFrontEndPreview, True Float=&RandomRangeFloat(Minimum=-20, Maximum=0), False Float=0)
      Z=50
     )
    )
   )
   &AddVelocityFromPoint(
    Module.Use Falloff Distance=false
    Origin Offset=<0 -600 0>
    Velocity Strength=&Multiply_Float(A=600, B=&MakeCustomFloatFromBool(Bool=$User.bIsFrontEndPreview, True Float=0, False Float=1))
   )
   &AddVelocity(Velocity=&Multiply_Vector(
     A=&RandomRangeVector(Minimum=<0 300 0>, Maximum=<0 700 0>)
     B=&MakeVector(X=1, Y=1, Z=1)
    ))
  ]
  particleUpdateScript [
   &SolveForcesAndVelocity(Clamp Velocity=false, Limit Acceleration=false)
   &ParticleState()
   &ScaleSpriteSize(Scale Factor=&Vector2DFromCurve(
     Vector2Curve=DataInterface() {propertyValues {XCurve {keys [<at 0 0.5 RCIM_Cubic>,<at 0.5 1 RCIM_Cubic>]}, YCurve {keys [<at 0 0 RCIM_Cubic>,<at 0.5 1 RCIM_Cubic>]}}}
     ScaleCurve=&Vector2DFromFloat(Value=1)
    ))
   &ScaleSpriteSizeBySpeed(Max Scale Factor=<1 50>, Min Scale Factor=<0.01 0.05>, Velocity Threshold=1)
   &ScaleColor(
    Scale Alpha=&FloatFromCurve(
     FloatCurve=DataInterface() {propertyValues {Curve {keys [
         <at 0 0 RCIM_Cubic>
         <at 0.3759 1>
         {arriveTangent=-2.221, interpMode=RCIM_Cubic, leaveTangent=-2.221, tangentMode=RCTM_User, time=1, value=0}
        ]}}}
     Scale Curve=&MakeCustomFloatFromBool(
      Bool=$User.bIsFrontEndPreview
      True Float=&MakeCustomFloatFromBool(Bool=$User.bIsFullyDeployed, True Float=1, False Float=0)
      False Float=&Lerp_Float(A=0, B=1, Alpha=$User.ForwardDot)
     )
    )
    Scale RGB=&VectorFromFloat(Value=1)
    ScaleA=true
    ScaleRGB=true
   )
  ]
  rendererClasses [&NiagaraSpriteRendererProperties]
  renderers [{rendererClass=&NiagaraSpriteRendererProperties}]
 }
 VBucks_Toon_Frontend {
  emitterSpawnScript [
   &NM_ToonShaderParameters(HighLight_Softness=0, RimLight_Width=0)
   &NM_OutlineParameters()
  ]
  emitterUpdateScript [
   &EmitterState(Life Cycle Mode=::NewEnumerator1, Loop Behavior=::NewEnumerator1, Loop Duration Mode=::NewEnumerator1, UseLoopDelay=false)
   &SpawnRate(
    Spawn Probability=&MakeCustomFloatFromBool(Bool=$User.bIsFrontEnd, True Float=1, False Float=0)
    SpawnRate=&FloatFromCurve(
     FloatCurve=DataInterface() {propertyValues {Curve {keys [
         <at 2.49 0 RCIM_Cubic>
         {arriveTangent=69.4346, interpMode=RCIM_Cubic, leaveTangent=-31.0331, tangentMode=RCTM_Break, time=2.5, value=20}
         <at 3.4 1 RCIM_Cubic>
        ]}}}
     CurveIndex=$Emitter.Age
     Scale Curve=&Multiply_Float(
      A=25
      B=&MakeCustomFloatFromBool(Bool=&SetBoolByFloatComparison(A=$System.ForwardAndRightDot, B=0.4), True Float=5, False Float=1)
     )
    )
    Use Spawn Probability=true
   )
   &SpawnBurst_Instantaneous(
    Spawn Count=300
    Spawn Probability=&MakeCustomFloatFromBool(Bool=$User.bIsFrontEnd, True Float=1, False Float=0)
    Spawn Time=2.5
    Use Spawn Probability=true
    UseLoopCountLimit=false
   )
  ]
  particleSpawnScript [
   &PartitionParticles()
   &InitializeParticle(
    Color=&SelectLinearColorFromArray(
     Array Sampling Mode=::NewEnumerator0
     Direct Array Index=?
     Color Selection Array=DataInterface() {propertyValues {ColorData [
        <rgba 0.2086 0.6172 0.3613 1>
        <rgba 0.8632 0.7682 0.0343 1>
        <rgba 0.9047 0.4287 0.1384 1>
        <rgba 0.8879 0.2705 0.6514 1>
        <rgba 0.956 0.2961 1 1>
        <rgba 0.4452 0.1248 0.6172 1>
       ]}}
     Random Seed=0
    )
    Lifetime=&RandomRangeFloat(Minimum=0.6, Maximum=1.5)
    Mesh Renderer Array Visibility Mode=::NewEnumerator3
    Mesh Renderer Info=DataInterface() {propertyValues {MeshRenderer=&self:VBucks_Toon_Frontend_2.NiagaraMeshRendererProperties_1}}
    Mesh Scale Mode=::NewEnumerator3
    Mesh Uniform Scale=1.25
    UsePositionOffset=false
    Write Material Random=true
   )
   &SkeletalMeshLocation(
    Bone / Socket Index=$Particles.PartitionParticles.Partition
    MaskSpawnGroup=false
    Mesh Sampling Type=::NewEnumerator5
    Sampled Position Offset=&MakeCustomVectorFromBool(Bool=$User.bIsFrontEnd, True Vector=<0 0 0>, False Vector=<0 0 0>)
    Skeletal Mesh=$User.SkeletalMesh
    Socket Sampling Mode=::NewEnumerator5
   )
   &ShapeLocation(
    Box Size=&MakeVector(X=5, Y=75, Z=30)
    MaskSpawnGroup=false
    Override Seed=false
    Shape Primitive=::NewEnumerator2
    Surface Only=false
    UseOwnerScale=true
   )
   &AddVelocity(
    Cone Angle=1
    Cone Axis=&RotateVectorByQuaternion(
     Vector To Rotate=&SelectVectorFromArray(
      Direct Array Index=$Particles.PartitionParticles.Partition
      Vector Selection Array=DataInterface() {propertyValues {FloatData [<-1 0 0>, <1 0 0>]}}
     )
     Quaternion=$Particles.SkeletalMeshLocation.SampledRotation
    )
    Module.UseConeFalloff=false
    Module.UseVelDistribution=false
    Override Seed=false
    Rotation Coordinate Space=::World
    Velocity Mode=::NewEnumerator2
    Velocity Speed=&Multiply_Float(
     A=&RandomRangeFloat(Minimum=500, Maximum=1000)
     B=&FloatFromCurve(
      FloatCurve=DataInterface() {propertyValues {Curve {keys [
          <at 2.2 0 RCIM_Cubic>
          {arriveTangent=1.9608, interpMode=RCIM_Cubic, leaveTangent=-3.6825, tangentMode=RCTM_Break, time=2.3, value=2}
          <at 3 1 RCIM_Cubic>
         ]}}}
      CurveIndex=$Emitter.Age
     )
    )
   ) {enabled=false}
   &AddVelocity(
    Rotation Coordinate Space=::World
    Velocity=&RotateVectorByQuaternion(
     Vector To Rotate=&SelectVectorFromArray(
      Direct Array Index=$Particles.PartitionParticles.Partition
      Vector Selection Array=DataInterface() {propertyValues {FloatData [<-1 0 0>, <1 0 0>]}}
     )
     Quaternion=$Particles.SkeletalMeshLocation.SampledRotation
    )
    Velocity Speed Scale=&Multiply_Float(
     A=&RandomRangeFloat(Minimum=300, Maximum=600)
     B=&FloatFromCurve(
      FloatCurve=DataInterface() {propertyValues {Curve {keys [
          <at 2.4 0 RCIM_Cubic>
          {arriveTangent=0.0791, interpMode=RCIM_Cubic, leaveTangent=0.0791, tangentMode=RCTM_User, time=2.5, value=3}
          <at 3.012 1>
         ]}}}
      CurveIndex=$Emitter.Age
     )
    )
   ) {module=AddVelocity001}
   &InitialMeshOrientation(Use Rotation Vector=false)
  ]
  particleUpdateScript [
   &ParticleState()
   &ScaleMeshSize(Scale Factor=&VectorFromFloat(Value=&FloatFromCurve(FloatCurve=DataInterface() {propertyValues {Curve {keys [
          {arriveTangent=4.1785, interpMode=RCIM_Cubic, tangentMode=RCTM_Break, time=0, value=0}
          {arriveTangent=0.0224, interpMode=RCIM_Cubic, leaveTangent=0.0224, tangentMode=RCTM_Break, time=0.05, value=1}
          {interpMode=RCIM_Cubic, tangentMode=RCTM_Break, time=0.7, value=1}
          {interpMode=RCIM_Cubic, tangentMode=RCTM_Break, time=1, value=0}
         ]}}})))
   &UpdateMeshOrientation(Rotation Rate=&RandomRangeFloat(Minimum=-0.1, Maximum=0.1))
   &GravityForce()
   &GravityForce(Gravity=&RotateVectorByQuaternion(
     Vector To Rotate=&Multiply_VectorByFloat(
      Vector=<0 1 0>
      Float=&FloatFromCurve(
       FloatCurve=DataInterface() {propertyValues {Curve {keys [
           <at 0 0>
           {arriveTangent=0.0112, interpMode=RCIM_Cubic, leaveTangent=0.0112, tangentMode=RCTM_User, time=0.05, value=1.6}
           {arriveTangent=-0.0134, interpMode=RCIM_Cubic, leaveTangent=-0.0134, tangentMode=RCTM_User, time=1, value=0}
          ]}}}
       Scale Curve=15000
      )
     )
     Quaternion=$Particles.SkeletalMeshLocation.SampledRotation
    )) {module=GravityForce001}
   &AerodynamicDrag(
    Aerodynamic Rotational Drag=5
    Pivot Offset Maximum:Vector3f=<0.25 0.25 0.25>
    Pivot Offset Minimum:Vector3f=<-0.25 -0.25 -0.25>
    Pivot Offset Minimum Length:Vector3f=<0 0 0>
    Use Linear Drag=true
    Use Rotational Drag=true
   )
   &InheritSourceMovement(
    Applied Position Scale=&VectorFromFloat(Value=&FloatFromCurve(FloatCurve=DataInterface() {propertyValues {Curve {keys [{interpMode=RCIM_Cubic, tangentMode=RCTM_User, time=0.1, value=1}, {interpMode=RCIM_Cubic, tangentMode=RCTM_User, time=0.7, value=0}]}}}))
    ApplyForce=false
    ApplyPosition=true
    ApplyPositionFalloff=false
   ) {enabled=false}
   &SolveForcesAndVelocity(Clamp Velocity=false, Limit Acceleration=false)
   &AlignSpriteToMeshOrientation(SetAlignment=true, SetFacing=true)
   &self:SphereBlockingVolume(Sphere Origin:NiagaraPosition=$Engine.Emitter.SimulationPosition, Sphere Radius=150) {enabled=false, module=SphereBlockingVolume}
  ]
  rendererClasses [&NiagaraMeshRendererProperties]
  renderers [{rendererClass=&NiagaraMeshRendererProperties, rendererIndex=0}]
 }
 VBucks_Toon_Gameplay {
  emitterSpawnScript [
   &NM_ToonShaderParameters(HighLight_Softness=0, RimLight_Width=0)
   &NM_OutlineParameters()
  ]
  emitterUpdateScript [
   &EmitterState(Life Cycle Mode=::NewEnumerator1, Loop Behavior=::NewEnumerator1, Loop Duration Mode=::NewEnumerator1, UseLoopDelay=false)
   &SpawnRate(
    Spawn Probability=&MakeCustomFloatFromBool(Bool=$User.bIsFrontEnd, True Float=0, False Float=0.7)
    SpawnRate=&FloatFromCurve(
     FloatCurve=DataInterface() {propertyValues {Curve {keys [
         <at 2.49 0 RCIM_Cubic>
         {arriveTangent=69.4346, interpMode=RCIM_Cubic, leaveTangent=-31.0331, tangentMode=RCTM_Break, time=2.5, value=20}
         <at 3.4 1 RCIM_Cubic>
        ]}}}
     CurveIndex=$Emitter.Age
     Scale Curve=&Multiply_Float(
      A=25
      B=&MakeCustomFloatFromBool(Bool=&SetBoolByFloatComparison(A=$System.ForwardAndRightDot, B=0.4), True Float=5, False Float=1)
     )
    )
    Use Spawn Probability=true
   )
   &SpawnBurst_Instantaneous(
    Spawn Count=200
    Spawn Probability=&MakeCustomFloatFromBool(Bool=$User.bIsFrontEnd, True Float=0, False Float=1)
    Spawn Time=2.5
    Use Spawn Probability=true
    UseLoopCountLimit=false
   )
  ]
  particleSpawnScript [
   &PartitionParticles()
   &InitializeParticle(
    Color=&MultiplyLinearColor(
     Linear Color=&SelectLinearColorFromArray(
      Array Sampling Mode=::NewEnumerator0
      Direct Array Index=?
      Color Selection Array=DataInterface() {propertyValues {ColorData [
         <rgba 0.2086 0.6172 0.3613 1>
         <rgba 0.8632 0.7682 0.0343 1>
         <rgba 0.9047 0.4287 0.1384 1>
         <rgba 0.8879 0.2705 0.6514 1>
         <rgba 0.956 0.2961 1 1>
         <rgba 0.4452 0.1248 0.6172 1>
        ]}}
      Random Seed=0
     )
     Scale Factor=&MakeVector4_FromLinearColor(LinearColor=&FNTimeOfDayColor(Night Color=<rgba 0.2855 0.3326 0.474 1>, Morning Color=<rgba 1 1 1 1>, Day Color=<rgba 1 1 1 1>, Evening Color=<rgba 1 1 1 1>))
    )
    Lifetime=&RandomRangeFloat(Minimum=1.8, Maximum=2.2)
    Mesh Renderer Array Visibility Mode=::NewEnumerator3
    Mesh Renderer Info=DataInterface() {propertyValues {MeshRenderer=&self:VBucks_Toon_Gameplay_3.NiagaraMeshRendererProperties_1}}
    Mesh Scale Mode=::NewEnumerator3
    Mesh Uniform Scale=1.25
    UsePositionOffset=false
    Write Material Random=true
   )
   &SkeletalMeshLocation(
    Bone / Socket Index=$Particles.PartitionParticles.Partition
    MaskSpawnGroup=false
    Mesh Sampling Type=::NewEnumerator5
    Sampled Position Offset=&MakeCustomVectorFromBool(Bool=$User.bIsFrontEnd, True Vector=<0 0 0>, False Vector=<0 0 0>)
    Skeletal Mesh=$User.SkeletalMesh
    Socket Sampling Mode=::NewEnumerator5
   )
   &ShapeLocation(
    Box Size=&MakeVector(X=5, Y=75, Z=30)
    MaskSpawnGroup=false
    Override Seed=false
    Shape Primitive=::NewEnumerator2
    Surface Only=false
    UseOwnerScale=true
   )
   &AddVelocity(
    Cone Angle=0
    Cone Axis=&RotateVectorByQuaternion(
     Vector To Rotate=&SelectVectorFromArray(
      Direct Array Index=$Particles.PartitionParticles.Partition
      Vector Selection Array=DataInterface() {propertyValues {FloatData [<1 0 0>, <-1 0 0>]}}
     )
     Quaternion=$Particles.SkeletalMeshLocation.SampledRotation
    )
    Module.UseConeFalloff=false
    Module.UseVelDistribution=false
    Override Seed=false
    Rotation Coordinate Space=::World
    Velocity Mode=::NewEnumerator2
    Velocity Speed=&Multiply_Float(
     A=&RandomRangeFloat(Minimum=500, Maximum=1000)
     B=&FloatFromCurve(
      FloatCurve=DataInterface() {propertyValues {Curve {keys [
          <at 2.2 0 RCIM_Cubic>
          {arriveTangent=0.0791, interpMode=RCIM_Cubic, leaveTangent=0.0791, tangentMode=RCTM_User, time=2.4, value=3}
          <at 3 1.75 RCIM_Cubic>
         ]}}}
      CurveIndex=$Emitter.Age
     )
    )
   ) {enabled=false}
   &AddVelocity(
    Rotation Coordinate Space=::World
    Velocity=&RotateVectorByQuaternion(
     Vector To Rotate=&SelectVectorFromArray(
      Direct Array Index=$Particles.PartitionParticles.Partition
      Vector Selection Array=DataInterface() {propertyValues {FloatData [<-1 0 0>, <1 0 0>]}}
     )
     Quaternion=$Particles.SkeletalMeshLocation.SampledRotation
    )
    Velocity Speed Scale=&Multiply_Float(
     A=&RandomRangeFloat(Minimum=500, Maximum=600)
     B=&FloatFromCurve(
      FloatCurve=DataInterface() {propertyValues {Curve {keys [
          <at 2.4 0 RCIM_Cubic>
          {arriveTangent=0.0791, interpMode=RCIM_Cubic, leaveTangent=0.0791, tangentMode=RCTM_User, time=2.5, value=4}
          <at 3 1.75 RCIM_Cubic>
         ]}}}
      CurveIndex=$Emitter.Age
     )
    )
   ) {module=AddVelocity001}
   &InitialMeshOrientation(Use Rotation Vector=false)
  ]
  particleUpdateScript [
   &ParticleState()
   &ScaleMeshSize(Scale Factor=&VectorFromFloat(Value=&FloatFromCurve(FloatCurve=DataInterface() {propertyValues {Curve {keys [
          {arriveTangent=4.1785, interpMode=RCIM_Cubic, tangentMode=RCTM_Break, time=0, value=0}
          {arriveTangent=0.0224, interpMode=RCIM_Cubic, leaveTangent=0.0224, tangentMode=RCTM_Break, time=0.05, value=1}
          {interpMode=RCIM_Cubic, tangentMode=RCTM_Break, time=0.7, value=1}
          {interpMode=RCIM_Cubic, tangentMode=RCTM_Break, time=1, value=0}
         ]}}})))
   &UpdateMeshOrientation(Rotation Rate=&RandomRangeFloat(Minimum=-0.1, Maximum=0.1))
   &GravityForce()
   &GravityForce(Gravity=&RotateVectorByQuaternion(
     Vector To Rotate=&Multiply_VectorByFloat(
      Vector=<0 1 0>
      Float=&FloatFromCurve(
       FloatCurve=DataInterface() {propertyValues {Curve {keys [
           {arriveTangent=0.0255, interpMode=RCIM_Cubic, leaveTangent=0.0255, tangentMode=RCTM_User, time=0, value=0}
           {arriveTangent=0.0112, interpMode=RCIM_Cubic, leaveTangent=0.0112, tangentMode=RCTM_User, time=0.1, value=1.6}
           {arriveTangent=-0.0134, interpMode=RCIM_Cubic, leaveTangent=-0.0134, tangentMode=RCTM_User, time=1, value=0}
          ]}}}
       Scale Curve=10000
      )
     )
     Quaternion=$Particles.SkeletalMeshLocation.SampledRotation
    )) {module=GravityForce001}
   &AerodynamicDrag(Aerodynamic Rotational Drag=5, Rotation Strength=2, Use Linear Drag=true, Use Rotational Drag=true)
   &InheritSourceMovement(
    Applied Position Scale=&VectorFromFloat(Value=&FloatFromCurve(FloatCurve=DataInterface() {propertyValues {Curve {keys [{interpMode=RCIM_Cubic, tangentMode=RCTM_User, time=0.1, value=1}, {interpMode=RCIM_Cubic, tangentMode=RCTM_User, time=0.7, value=0}]}}}))
    ApplyForce=false
    ApplyPosition=true
    ApplyPositionFalloff=false
   )
   &SolveForcesAndVelocity(Clamp Velocity=false, Limit Acceleration=false)
   &AlignSpriteToMeshOrientation(SetAlignment=true, SetFacing=true)
   &self:SphereBlockingVolume(Sphere Origin:NiagaraPosition=$Engine.Emitter.SimulationPosition, Sphere Radius=150) {enabled=false, module=SphereBlockingVolume}
  ]
  rendererClasses [&NiagaraMeshRendererProperties]
  renderers [{rendererClass=&NiagaraMeshRendererProperties, rendererIndex=0}]
 }
}
summary {
 systemName=NS_Glider_Lettuce
 emitters [
  {emitterName=VBucks_Toon_Gameplay, rendererClasses [&NiagaraMeshRendererProperties]}
  {emitterName=Smoke_Toon_WorldSpace, rendererClasses [&NiagaraMeshRendererProperties]}
  {emitterName=SpeedLines, rendererClasses [&NiagaraSpriteRendererProperties]}
  {emitterName=VBucks_Toon_Frontend, rendererClasses [&NiagaraMeshRendererProperties]}
  {emitterName=Paper_Pieces_Break, rendererClasses [&NiagaraMeshRendererProperties]}
  {emitterName=Paper_Pieces_Door_Right, rendererClasses [&NiagaraMeshRendererProperties]}
  {emitterName=Paper_Pieces_Exhaust_L, rendererClasses [&NiagaraMeshRendererProperties]}
  {emitterName=Paper_Pieces_Exhaust_R, rendererClasses [&NiagaraMeshRendererProperties]}
  {emitterName=Paper_Pieces_Headlight, rendererClasses [&NiagaraMeshRendererProperties]}
  {emitterName=Back_Blinkers, rendererClasses [&NiagaraSpriteRendererProperties, &NiagaraMeshRendererProperties]}
  {emitterName=Front_Blinkers, rendererClasses [&NiagaraSpriteRendererProperties, &NiagaraMeshRendererProperties]}
 ]
 userVariables [
  User.bIsFrontEnd {type=NiagaraBool}
  User.bIsFrontEndPreview {type=NiagaraBool}
  User.bIsFullyDeployed {type=NiagaraBool}
  User.ForwardDot {type=NiagaraFloat, defaultValue=0, description=""}
  User.RibbonIDOffset {type=NiagaraInt32}
  User.RightDot {type=NiagaraFloat, defaultValue=0, description=""}
  User.SkeletalMesh {
   type=NiagaraDataInterfaceSkeletalMesh
   defaultValue {
    struct=NiagaraExt_VariableValue_DataInterface
    value {dataInterface=&self:NiagaraDataInterfaceSkeletalMesh_0}
   }
  }
 ]
}
